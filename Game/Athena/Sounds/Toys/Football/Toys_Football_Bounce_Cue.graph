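# Sound cue for the bounce of the toy football in Athena.
# A Random node picks one of two equally weighted bounce waves, then a Modulator varies its pitch.
# It plays at reduced volume with the shared ball-bounce attenuation asset and is routed to the emote music sound class.
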SoundCue /Game/Athena/Sounds/Toys/Football/Toys_Football_Bounce_Cue
root=Modulator_0
nodes [
 {id=Modulator_0, children [Random_1], properties {pitchMin=0.85}}
 {id=Random_1, children [WavePlayer_0,WavePlayer_1], properties {weights [1,1]}}
 {id=WavePlayer_0, properties {soundWaveAssetPtr=&/Toy_Football_Bounce_01}}
 {id=WavePlayer_1, properties {soundWaveAssetPtr=&/Toy_Football_Bounce_02}}
]
properties {
 attenuationSettings=/Game/Sounds/Attenuation/Fort_Ball_Bounce_Athena_Att
 maxDistance=7000
 soundClassObject=&Fort_Music_Emote
 volumeMultiplier=0.3
}
